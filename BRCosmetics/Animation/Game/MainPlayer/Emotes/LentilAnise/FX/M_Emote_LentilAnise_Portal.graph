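# Masked, two-sided unlit Niagara sprite material for the portal effect of the LentilAnise emote.
# An inner-to-outer color gradient combines with three swirl layers: a thresholded noise-cutout swirl and two radially masked, panning splatter textures.
# Panning polar noise distorts a radial gradient for opacity, and the color switches to an outline color below a threshold.
# Panned samples use derivatives from the original UVs for mip selection, avoiding polar-coordinate seams.
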Material /BRCosmetics/Animation/Game/MainPlayer/Emotes/LentilAnise/FX/M_Emote_LentilAnise_Portal
graph {
 properties {
  bAutomaticallySetUsageInEditor=false
  bUsedWithNiagaraSprites=true
  blendMode=BLEND_Masked
  shadingModel=MSM_Unlit
  twoSided=true
  usedShadingModels=Unlit
 }
 nodes [
  MaterialFunctionCall(Emissive=NamedRerouteUsage()) {materialFunction=&MF_FX_EmissiveAdjustments}
  n1 = MaterialFunctionCall!(UV TexCoord=TextureCoordinate(), Time, Speed=Constant() {r=0.01}, Spin=Constant() {r=0.7}) {materialFunction=/Game/Effects/Niagara/MasterMaterials/Functions/Mojo_MF/MF_FX_Swirl_UV}
  n2 = Multiply(A=Add(A=@n1.Out UV TexCoord) {constB=0.2}) {constB=4}
  NamedRerouteDeclaration(None=TextureSample(UVs=@n2) {texture=/Game/Effects/Fort_Effects/Textures/Misc/T_Shot0030_NoiseCutout}.G) {name=Main Swirl Tex, nodeColor=<rgba 1 0.375 0 1>}
  n3 = NamedRerouteUsage()
  n4 = MaterialFunctionCall!(UV Coord, LenghtScale=Constant() {r=2.7}, RadialScale) {materialFunction=/Game/Global/MaterialFunctions/TechArt/MF_PolarCoordinates}
  n5 = LinearInterpolate(
   A=Constant3Vector() {constant=<rgba 0.0337 0.2031 0.0554 1>}
   B=Constant3Vector() {constant=<rgba 0.1913 0.9304 0.1304 1>}
   Alpha=SmoothStep(Value=ComponentMask(None=@n4.Result) {r=true}) {constMin=0.8}
  ) {constAlpha=0}
  NamedRerouteDeclaration(None=If(A=Multiply(A=@n3) {constB=1.3}, "A > B"=@n5, "A == B"=@n5, "A < B"=Constant()) {constB=0.8}) {name=Swirl Layer 1, nodeColor=<rgba 0 1 0.1328 1>}
  n6 = TextureCoordinate()
  n7 = DDY(Value=@n6)
  NamedRerouteDeclaration(None=@n7) {name=DDY, nodeColor=<rgba 1 0 0.9141 1>}
  n8 = ComponentMask(None=MaterialFunctionCall() {materialFunction=&MF_PolarCoordinates}.Result) {r=true}
  n9 = Multiply(A=SmoothStep(Value=@n8) {constMin=0.1}, B=SmoothStep(Value=OneMinus(None=@n8)) {constMax=0.9, constMin=0.6})
  n10 = NamedRerouteUsage()
  n11 = TextureSample!(
   UVs=Multiply(A=Panner(Coordinate=@n10) {speedX=0.1, speedY=0.1}) {constB=2}
   Tex
   "DDX(UVs)"=NamedRerouteUsage()
   "DDY(UVs)"=NamedRerouteUsage()
   Apply View MipBias
  ) {mipValueMode=TMVM_Derivative, texture=/Game/Global/Textures/Default/PointSplatter/T_UI_StarBurst_Scatter}
  NamedRerouteDeclaration(None=Multiply(A=VectorParameter() {parameterName=SwirlLayer3Color, defaultValue=<rgba 967.3938 1000 326.0201 1>}.RGB, B=Multiply(A=@n9, B=@n11.R))) {name=Swirl Layer 3, nodeColor=<rgba 1 0 0 1>}
  n12 = MaterialFunctionCall(UV Coord=@n6, LenghtScale=Constant() {r=1}, RadialScale=Constant() {r=1}) {materialFunction=&MF_PolarCoordinates}
  n13 = DDX(Value=@n6)
  n14 = TextureSample(UVs=Panner(Coordinate=@n12.Result) {speedX=-0.3, speedY=0.5}, "DDX(UVs)"=@n13, "DDY(UVs)"=@n7) {mipValueMode=TMVM_Derivative, texture=/Game/Global/Textures/Default/Noise/T_SoftNoise}
  NamedRerouteDeclaration(None=@n14.R) {name=Distortion, nodeColor=<rgba 0.7813 0 1 1>}
  n15 = NamedRerouteUsage()
  n16 = NamedRerouteUsage()
  n17 = ComponentMask(None=@n16) {r=true}
  n18 = NamedRerouteUsage()
  n19 = Add(
   A=If(A=@n15, "A > B"=@n15, "A == B"=@n15, "A < B"=If(A=@n17, "A > B"=@n16, "A == B"=@n16, "A < B"=@n18) {constB=0.001}) {constB=0.2}
   B=NamedRerouteUsage()
  )
  NamedRerouteDeclaration(None=If(
    A=NamedRerouteUsage()
    "A > B"=@n19
    "A == B"=@n19
    "A < B"=VectorParameter() {parameterName=OutlineColor, defaultValue=<rgba 0.1304 0.3304 0.0661 1>}.RGB
   ) {constB=0.4}) {name=Color, nodeColor=<rgba 0 1 0.5313 1>}
  n20 = VectorParameter() {parameterName=BaseColorOuter, defaultValue=<rgba 0.0881 0.3438 0.0299 1>}
  n21 = VectorParameter() {parameterName=BaseColorInner, defaultValue=<rgba 0.0522 0.3304 0.0435 1>}
  n22 = MaterialFunctionCall(LenghtScale=Constant() {r=3}) {materialFunction=&MF_PolarCoordinates}
  n23 = NamedRerouteUsage()
  n24 = ComponentMask(None=Add(A=@n22.Result, B=Multiply(A=@n23) {constB=0.8})) {r=true}
  NamedRerouteDeclaration(None=LinearInterpolate(A=@n20.RGB, B=LinearInterpolate(A=@n21.RGB, B=@n20.RGB, Alpha=@n24), Alpha=@n22.Result)) {name=Base Portal, nodeColor=<rgba 0 1 0.5078 1>}
  n25 = Multiply(A=@n14.R) {constB=0.05}
  NamedRerouteDeclaration(None=MaterialFunctionCall(UVs=Add(A=@n6, B=@n25), Radius=Constant() {r=0.6}) {materialFunction=&RadialGradientExponential}.RadialGradientExponential) {name=Opacity, nodeColor=<rgba 0.1016 0 1 1>}
  NamedRerouteDeclaration(None=@n2) {name=Swirl UVs, nodeColor=<rgba 1 0.6563 0 1>}
  NamedRerouteDeclaration(None=@n12.Result) {name=Polar UVs, nodeColor=<rgba 1 0 0.8672 1>}
  n26 = SmoothStep(Value=ComponentMask(None=MaterialFunctionCall() {materialFunction=&MF_PolarCoordinates}.Result) {r=true}) {constMax=0.4, constMin=0.3}
  n27 = TextureSample(UVs=Panner(Coordinate=NamedRerouteUsage()) {speedX=0.05, speedY=0.2}, "DDX(UVs)"=NamedRerouteUsage(), "DDY(UVs)"=NamedRerouteUsage()) {mipValueMode=TMVM_Derivative, texture=/Game/Global/Textures/Default/PointSplatter/BoatFlake}
  NamedRerouteDeclaration(None=Multiply(A=VectorParameter() {parameterName=SwirlLayer2Color, defaultValue=<rgba 4.7447 5 4.6838 1>}.RGB, B=Multiply(A=@n26, B=@n27.R))) {name=Swirl Layer 2, nodeColor=<rgba 1 0 0 1>}
  NamedRerouteDeclaration(None=@n13) {name=DDX, nodeColor=<rgba 0 0.2734 1 1>}
  NamedRerouteUsage()
  Panner(Coordinate=MaterialFunctionCall(UV Coord=@n1.Out UV TexCoord, LenghtScale=Constant() {r=1}, RadialScale=Constant() {r=3}) {materialFunction=&MF_PolarCoordinates}.Result) {speedX=0.1}
 ]
}
