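# Niagara effect for the LentilAnise ClimbSpill emote's prop spawn, centered on a blue hologram-style sprite named Rick with flares and confetti.
# Rick plays curve-driven flipbook frames, turns toward the camera horizontally, and fades by view angle alongside a flat cover sprite.
# A sprite and mesh flash bursts near the end of the roughly 3.7-second sequence.
# Confetti collision modules are disabled; beam and glitch-box emitters have no renderers.
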NiagaraSystem /BRCosmetics/Animation/Game/MainPlayer/Emotes/LentilAnise/FX/NS_Emote_ClimbSpill_PropSpawn
dependencies {
 usedDataInterfaces [
  &NiagaraDataInterfaceArrayColor
  &NiagaraDataInterfacePlatformSet
  &NiagaraDataInterfaceCurve
  &NiagaraDataInterfaceArrayFloat
  &NiagaraDataInterfaceVector2DCurve
  &NiagaraDataInterfaceCamera
  &NiagaraDataInterfaceSpriteRendererInfo
  &NiagaraDataInterfaceColorCurve
 ]
 usedDynamicInputs [&FNTimeOfDayColor]
 usedModules [
  &SystemState
  &BlendSpaceHelper
  &self:SystemSpawnScript.SystemScriptSource.SystemScriptGraph.NiagaraNodeAssignment_0.SetVariables_#1
 ]
}
emitters {
 Beam_LWE {rendererClasses [], renderers []}
 Confetti {
  emitterUpdateScript [
   &EmitterState(Life Cycle Mode=::NewEnumerator1, Loop Behavior=::NewEnumerator1, Loop Duration=3, UseLoopDelay=true)
   &SpawnBurst_Instantaneous(Spawn Count=100, Spawn Group=1, Use Spawn Probability=false, UseLoopCountLimit=false) {module=SpawnBurst_Instantaneous001}
   &SpawnBurst_Instantaneous(Spawn Count=40, Use Spawn Probability=false, UseLoopCountLimit=false)
  ]
  particleSpawnScript [
   /SetVariables(Particles.VisibilityTag=&SpawnGroup()) {isSetParameters=true}
   &InitializeParticle(
    Color=&SelectLinearColorFromArray(
     Array Sampling Mode=::NewEnumerator0
     Direct Array Index=?
     Color Selection Array=DataInterface() {propertyValues {ColorData [<rgba 0 1 0.6172 1>, <rgba 0.3536 0.1041 0.855 1>, <rgba 0.8174 0.7262 0.2317 1>, <rgba 0.8783 0.1421 0.2757 1>]}}
     Random Seed=0
    )
    Lifetime=&MergeHelper_FloatSwitch(
     0=&RandomRangeFloat(Minimum=0.5, Maximum=1.5)
     1=&RandomRangeFloat(Minimum=3, Maximum=2)
     2=?
     3=?
     Index=$Particles.VisibilityTag
     Num=::NewEnumerator0
    )
    Sprite Size Max=<1 3>
    Sprite Size Min=<3 7>
    Sprite Size Mode=::NewEnumerator2
    UsePositionOffset=false
    Write Material Random=true
   )
   &ShapeLocation(
    Hemisphere Distribution=<1 0.5>
    MaskSpawnGroup=false
    Non Uniform Scale=<1 1 2>
    Override Seed=false
    Sphere Radius=5
    Sphere Surface Distribution=1
    UseOwnerScale=true
   )
   &AddVelocity(
    Cone Angle=70
    Cone Axis=<0 0 1>
    Module.UseConeFalloff=false
    Module.UseVelDistribution=false
    Override Seed=false
    Velocity Mode=::NewEnumerator2
    Velocity Speed=&RandomRangeFloat(Minimum=2000, Maximum=500)
   )
  ]
  particleUpdateScript [
   &ParticleState()
   &GravityForce(Gravity=&RandomRangeVector(Minimum=<0 0 -400>, Maximum=<0 0 -600>))
   &Collision(
    Age Colliding Particles=false
    Collision Enabled=&Boolean_AndOperation(
     Bool1=&Boolean_AndOperation(
      Bool1=&IsPlatformSetActive(New Platform Set=DataInterface() {propertyValues {Platforms {qualityLevelMask=-8}}})
      Bool2=&Boolean_NotOperation(Bool=$User.bIsFrontEndPreview)
     )
     Bool2=&SetBoolByFloatComparison(
      A=&FloatFromCurve(FloatCurve=DataInterface() {propertyValues {Curve {
          defaultValue=1
          keys [{interpMode=RCIM_Cubic, tangentMode=RCTM_User, time=0.2, value=0}, {interpMode=RCIM_Cubic, tangentMode=RCTM_User, time=0.3, value=1}]
         }}})
      B=0.5
     )
    )
    Control Roll On Collision=true
    Enable Rest State=false
    Friction=0.1
    KillOnCollision=false
    Module.EnableMaxCollisionCount=false
    Module.Randomize Collision Normal Vector=false
    Particle Radius Scale=5
    Radius Calculation Type=::NewEnumerator1
    Restitution=0.4
    ScaleRateOfRotation=true
   ) {enabled=false}
   &Collision(
    Age Colliding Particles=false
    Analytical Collision Plane Space:ENiagaraCoordinateSpace=::Local
    CPU Collision Type=::NewEnumerator3
    Collision Enabled=&Boolean_AndOperation(
     Bool1=&Boolean_AndOperation(
      Bool1=&IsPlatformSetActive(New Platform Set=DataInterface() {propertyValues {Platforms {qualityLevelMask=-8}}})
      Bool2=&Boolean_NotOperation(Bool=$User.bIsFrontEndPreview)
     )
     Bool2=&SetBoolByFloatComparison(
      A=&FloatFromCurve(FloatCurve=DataInterface() {propertyValues {Curve {
          defaultValue=1
          keys [{interpMode=RCIM_Cubic, tangentMode=RCTM_User, time=0.2, value=0}, {interpMode=RCIM_Cubic, tangentMode=RCTM_User, time=0.3, value=1}]
         }}})
      B=0.5
     )
    )
    Control Roll On Collision=true
    Enable Rest State=false
    Friction=0.1
    KillOnCollision=false
    Module.EnableMaxCollisionCount=false
    Module.Randomize Collision Normal Vector=false
    Particle Radius Scale=5
    Perform 2 Analytical Collision Queries=false
    Radius Calculation Type=::NewEnumerator1
    Restitution=0.8
    ScaleRateOfRotation=false
   ) {enabled=false, module=Collision002}
   &CurlNoiseForce(
    Bias Noise=false
    Noise Frequency=20
    Noise Strength=&MakeCustomFloatFromBool(Bool=$Particles.HasCollided, True Float=50, False Float=500)
    Pan Noise=false
   )
   &SpriteRotationRate(Rotation Rate=&MakeCustomFloatFromBool(
     Bool=$Particles.HasCollided
     True Float=0
     False Float=&SelectFloatFromArray(
      Array Sampling Mode=::NewEnumerator0
      Direct Array Index=?
      Float Selection Array=DataInterface() {propertyValues {FloatData [-2000,-2000,2000,-2000,2000,-2000]}}
      Random Seed=0
     )
    ))
   &Drag(
    Drag=&FloatFromCurve(
     FloatCurve=DataInterface() {propertyValues {Curve {
        defaultValue=1
        keys [{tangentMode=RCTM_User, time=0.05, value=0.5}, {tangentMode=RCTM_User, time=0.07, value=1}, <at 0.15 1>, <at 0.3961 0.6>, <at 0.7937 0.9136>]
       }}}
     Scale Curve=&RandomRangeFloat(Minimum=10, Maximum=20)
    )
    Use Linear Drag=true
    Use Rotational Drag=true
   )
   &SolveForcesAndVelocity(Clamp Velocity=false, Limit Acceleration=false)
   &AlignSpriteToMeshOrientation(SetAlignment=true, SetFacing=true) {enabled=false}
   &ScaleSpriteSizeBySpeed(Max Scale Factor=&MaskVector2dBySpawnGroup(
     Masked Vector2D=&Vector2DFromCurve(
      Vector2Curve=DataInterface() {propertyValues {
        XCurve {keys [<at 0 1>,<at 1 1>]}
        YCurve {keys [
          <at 0.0016 1 RCIM_Cubic>
          {arriveTangent=-45.1734, interpMode=RCIM_Cubic, leaveTangent=-45.1734, tangentMode=RCTM_User, time=0.2152, value=10}
          <at 0.6051 1 RCIM_Cubic>
         ]}
       }}
      ScaleCurve=<1 1>
     )
     Passthrough Vector2D=<1 1>
     Spawn Group Mask=0
    ))
   &ScaleSpriteSize(
    Scale Sprite Size Mode=::NewEnumerator0
    Uniform Scale Factor=&Multiply_Float(
     A=&FloatFromCurve(FloatCurve=DataInterface() {propertyValues {Curve {
         defaultValue=0
         keys [
          {arriveTangent=-4.2555, interpMode=RCIM_Cubic, leaveTangent=-4.2555, tangentMode=RCTM_User, time=-0.0033, value=2}
          {interpMode=RCIM_Cubic, tangentMode=RCTM_User, time=0.7969, value=1}
          {arriveTangent=-3.0916, interpMode=RCIM_Cubic, leaveTangent=-3.0916, tangentMode=RCTM_User, time=1, value=0}
         ]
        }}})
     B=1
    )
   )
   &BlendSpace("Position Blend (World <-> Local)":NiagaraFloat=&MakeCustomFloatFromBool(
     Bool=&SetBoolByFloatComparison(A=&VectorLength(Vector=$Engine.Owner.Velocity), B=500)
     True Float=1
     False Float=0
     Translation Contribution:NiagaraFloat=1
     Rotation Contribution:NiagaraFloat=1
    )) {enabled=false}
  ]
  rendererClasses [&NiagaraSpriteRendererProperties]
  renderers [
   {rendererClass=&NiagaraSpriteRendererProperties}
   {rendererClass=&NiagaraSpriteRendererProperties, rendererIndex=1}
  ]
 }
 Flash001 {
  emitterUpdateScript [
   &EmitterState()
   &SpawnBurst_Instantaneous(Spawn Count=4, Spawn Time=3.55, Use Spawn Probability=false)
  ]
  particleSpawnScript [
   &PartitionParticles(Number Of Partitions=4)
   &InitializeParticle(
    Color=<rgba 0.1524 0.7275 1 0.5>
    Lifetime=0.35
    Position Offset=<0 0 102>
    Sprite Size Mode=::NewEnumerator3
    Uniform Sprite Size=&Multiply_Float(A=100, B=1)
    UsePositionOffset=true
    Write Material Random=true
   )
   &ShapeLocation(MaskSpawnGroup=false, Override Seed=false, Sphere Radius=0, UseOwnerScale=true)
   &CameraOffset(Camera Offset Amount=-10)
   /SetVariables(Particles.SizeWideFlare:Vector2f=&Multiply_Vector2D_ByFloat(Float=&Multiply_Float(A=1, B=1), Vector2D=<225 40>)) {isSetParameters=true}
   &InitialMeshOrientation(Mesh Orientation Mode=::NewEnumerator2, Orientation Vector=<0 1 0>, Rotation=<0.25 0 0>, Use Rotation Vector=true)
   &AlignSpriteToMeshOrientation(SetAlignment=true, SetFacing=true)
  ]
  particleUpdateScript [
   &ParticleState()
   &SolveForcesAndVelocity(Clamp Velocity=false, Limit Acceleration=false)
   &ScaleColor(
    Scale Alpha=&FloatFromCurve(FloatCurve=DataInterface() {propertyValues {Curve {keys [
         <at 0.1 0 RCIM_Cubic>
         {interpMode=RCIM_Cubic, leaveTangent=-2.0997, tangentMode=RCTM_Break, time=0.2521, value=1}
         {arriveTangentWeight=0.7369, interpMode=RCIM_Cubic, tangentMode=RCTM_Break, time=1, value=1}
        ]}}})
    Scale RGB=&VectorFromFloat(Value=&FloatFromCurve(
      FloatCurve=DataInterface() {propertyValues {Curve {keys [
          <at 0.1 0 RCIM_Cubic>
          {interpMode=RCIM_Cubic, leaveTangent=-2.0997, tangentMode=RCTM_Break, time=0.2521, value=1}
          {arriveTangentWeight=0.7369, interpMode=RCIM_Cubic, tangentMode=RCTM_Break, time=0.6, value=1}
         ]}}}
      Scale Curve=5
     ))
    ScaleA=true
    ScaleRGB=true
   )
   &ScaleSpriteSize(Uniform Curve Sprite Scale=DataInterface() {propertyValues {Curve {keys [
        <at 0.3 0 RCIM_Cubic>
        {interpMode=RCIM_Cubic, tangentMode=RCTM_User, time=0.4, value=1}
        {interpMode=RCIM_Cubic, tangentMode=RCTM_User, time=1, value=0}
       ]}}})
   &FN_Sprite_Attributes(
    Color=&MakeLinearColorFromVectorAndFloat(
     "Vector (RGB)"=&MakeVectorFromLinearColorRGB(Color=<rgba 0.3981 2.0918 3 1>)
     "Float (Alpha)"=&Multiply_Float(
      A=1
      B=&FloatFromCurve(FloatCurve=DataInterface() {propertyValues {Curve {keys [{interpMode=RCIM_Cubic, tangentMode=RCTM_User, time=0, value=1}, {interpMode=RCIM_Cubic, tangentMode=RCTM_User, time=1, value=0.5}]}}})
     )
    )
    Create Color Attribute=true
    Write to Module Namespace=true
   )
   /SetVariables(Particles.SizeWideFlare:Vector2f=&Multiply_Vector2D(
     A=$Particles.Initial.SizeWideFlare
     B=&MakeVector2D(
      X=&FloatFromCurve(FloatCurve=DataInterface() {propertyValues {Curve {keys [<at 0.1 0>,<at 0.2 0.4>,<at 0.3 1>,<at 1 0>]}}})
      Y=&FloatFromCurve(FloatCurve=DataInterface() {propertyValues {Curve {keys [<at 0.2 1>,<at 0.5 0.1>]}}})
     )
    )) {isSetParameters=true}
  ]
  rendererClasses [&NiagaraSpriteRendererProperties, &NiagaraMeshRendererProperties]
  renderers [
   {rendererClass=&NiagaraSpriteRendererProperties}
   {rendererClass=&NiagaraSpriteRendererProperties, rendererIndex=1}
   {rendererClass=&NiagaraMeshRendererProperties, rendererIndex=2}
  ]
 }
 GlitchBox {rendererClasses [], renderers []}
 GlitchBoxBottom {rendererClasses [], renderers []}
 Rick {
  emitterUpdateScript [
   &EmitterState(Inactive Response=::NewEnumerator1, Life Cycle Mode=::NewEnumerator1, Loop Behavior=::NewEnumerator1, Loop Duration=3.7, UseLoopDelay=false)
   &SpawnBurst_Instantaneous(Use Spawn Probability=false, UseLoopCountLimit=false)
   &SpawnBurst_Instantaneous(Spawn Count=5, Spawn Group=1, Use Spawn Probability=false, UseLoopCountLimit=false) {module=SpawnBurst_Instantaneous001}
   &SpawnBurst_Instantaneous(Spawn Count=2, Spawn Group=2, Use Spawn Probability=false, UseLoopCountLimit=false) {module=SpawnBurst_Instantaneous002}
   &SpawnRate(
    SpawnGroup=1
    SpawnRate=&FloatFromCurve(
     FloatCurve=DataInterface() {propertyValues {Curve {keys [<at 0 0>,<at 0.013 1>,<at 0.88 1>,<at 0.95 0>]}}}
     CurveIndex=$Emitter.NormalizedLoopAge
     Scale Curve=15
    )
    Use Spawn Probability=false
   )
  ]
  particleSpawnScript [
   &InitializeParticle(
    Color=<rgba 0.1302 0.7177 1 2>
    Lifetime=3.8
    Mesh Scale Mode=::NewEnumerator3
    Mesh Uniform Scale=5
    Position Offset=<0 0 110>
    Sprite Size=&Multiply_Vector2D_ByFloat(Float=100, Vector2D=<1 1>)
    Sprite Size Mode=::NewEnumerator4
    UsePositionOffset=true
    Write Material Random=true
   )
   /SetVariables(Particles.VisibilityTag=$Emitter.SpawnGroup) {isSetParameters=true}
   &InitialMeshOrientation(Mesh Orientation Mode=::NewEnumerator2, Orientation Vector=<1 -1 0>, Rotation=<0.25 0 0>, Use Rotation Vector=true) {enabled=false}
   &ApplyInitialForces() {enabled=false}
   &AlignSpriteToMeshOrientation(SetAlignment=true, SetFacing=true) {enabled=false}
   /SetVariables(
    Particles.BurstFlarePos:NiagaraPosition=&SubtractVectorFromPosition(Position=$Engine.Emitter.SimulationPosition, Vector=<0 0 -100>)
    Particles.BurstFlareSize:Vector2f=<111.1111 222>
    Particles.FlareHoloBaseCol:LinearColor=<rgba 0.9943 4.8798 7 1>
    Particles.FlareHoloBasePos:NiagaraPosition=&SubtractVectorFromPosition(Position=$Engine.Emitter.SimulationPosition, Vector=<0 0 -62>)
    Particles.FlareHoloBaseSize:Vector2f=<130 50>
    Particles.LongHoloColor:LinearColor=<rgba 0.2131 1.0457 1.5 1>
    Particles.LongHoloPos:NiagaraPosition=&SubtractVectorFromPosition(Position=$Engine.Emitter.SimulationPosition, Vector=<0 0 -20>)
    Particles.LongHoloSize:Vector2f=&RandomRangeVector2D(Minimum=<140 50>, Maximum=<100 250>)
   ) {isSetParameters=true}
  ]
  particleUpdateScript [
   &SolveForcesAndVelocity(Clamp Velocity=false, Limit Acceleration=false)
   &ParticleState()
   &ScaleSpriteSize(
    Non-Uniform Curve Index:NiagaraFloat=$Emitter.Age
    Non-Uniform Curve Sprite Scale=DataInterface() {propertyValues {
      XCurve {keys [
        {tangentMode=RCTM_User, time=0, value=0}
        <at 0.0296 0.1731>
        <at 0.1048 1.338>
        <at 0.1914 1>
        {arriveTangent=0.1774, interpMode=RCIM_Cubic, leaveTangent=0.1774, time=3.5743, value=1}
        {arriveTangent=-10.3087, interpMode=RCIM_Cubic, leaveTangent=-10.3087, time=3.6131, value=1.6071}
        <at 3.6713 0 RCIM_Cubic>
       ]}
      YCurve {keys [{tangentMode=RCTM_User, time=0, value=0}, <at 0.0467 1.4818>, <at 0.1242 0.8642>, <at 0.1914 1>, <at 0.9 1>]}
     }}
    Scale Sprite Size Mode=::NewEnumerator3
   )
   /SetVariables(Particles.CameraZ:NiagaraFloat=&RemapRange(
     Input Value To Remap=&Abs_Float(Float:NiagaraFloat=&MakeFloatFromVector(
       Vector=&Subtract_Vector(
        A=Linked() {linkedVariable {name=Particles.Position, type=NiagaraPosition}}
        B=&ReturnCameraPosition(Coordinate Space:ENiagaraCoordinateSpace=::World, Camera Query:NiagaraDataInterfaceCamera=DataInterface())
       )
       Channel=::NewEnumerator2
      ))
     Input Min=0
     Input Max=360
     Output Min=0
     Output Max=1
    )) {isSetParameters=true}
   &SubUVAnimation(
    "Sprite Renderer "=DataInterface() {propertyValues {SpriteRenderer=&self:Rick_0.NiagaraSpriteRendererProperties_0}}
    SubUV Lookup Index:NiagaraFloat=&FloatFromCurve(FloatCurve=DataInterface() {propertyValues {Curve {
        defaultValue=1
        keys [
         {tangentMode=RCTM_User, time=0, value=0}
         <at 0.1559 0.49>
         <at 0.1588 0.32>
         <at 0.2022 0.49>
         <at 0.2257 0.49>
         <at 0.2292 0.32>
         <at 0.2638 0.49>
         <at 0.2926 0.49>
         <at 0.2988 0.32>
         <at 0.3368 0.49>
         <at 0.3617 0.49>
         <at 0.3731 0.32>
         <at 0.4148 0.49>
         <at 0.5701 0.85>
         <at 0.5779 0.69>
         <at 0.625 0.85>
         <at 0.6628 0.85>
         <at 0.672 0.69>
         <at 0.726 0.85>
         <at 0.7613 0.85>
         <at 0.7734 0.69>
         <at 0.8207 0.85>
         <at 0.8524 0.85>
         <at 0.8607 0.69>
         <at 0.9145 0.85>
         <at 1 1.1>
        ]
       }}})
    UseEndFrame=false
    UseStartFrame=false
   )
   &ScaleColor(
    Scale Alpha=&FloatFromCurve(
     FloatCurve=DataInterface() {propertyValues {Curve {keys [<at 0 1>,<at 0.6 1>,<at 0.9804 0>]}}}
     CurveIndex=&Abs_Float(Float:NiagaraFloat=&DotProduct(
       Vector A=&ReturnCameraVector(
        Camera Vectors=::NewEnumerator4
        Coordinate Space=::Simulation
        Origin Position=$Particles.Position
        CameraQuery=DataInterface() {propertyValues {bRequireCurrentFrameData=false}}
       )
       Vector B=$Particles.SpriteAlignment
      ))
    )
    Scale RGB=&Multiply_VectorByFloat(
     Vector=<1 1 1>
     Float=&FloatFromCurve(FloatCurve=DataInterface() {propertyValues {Curve {
         defaultValue=1
         keys [
          {interpMode=RCIM_Cubic, tangentMode=RCTM_User, time=0, value=5}
          {interpMode=RCIM_Cubic, tangentMode=RCTM_User, time=0.06, value=1}
          {tangentMode=RCTM_User, time=1, value=1}
         ]
        }}})
    )
    ScaleA=true
    ScaleRGB=true
   )
   &UpdateMeshOrientation(
    Facing Interpolation=false
    Facing Position:NiagaraPosition=&Subtract_Vector(
     A=&ReturnCameraPosition(Coordinate Space:ENiagaraCoordinateSpace=::Simulation, Camera Query:NiagaraDataInterfaceCamera=DataInterface())
     B=&MakeVector(
      X=0
      Y=0
      Z=&Subtract_Float(
       A=&MakeFloatFromVector(
        Vector=&ReturnCameraPosition(Coordinate Space:ENiagaraCoordinateSpace=::Simulation, Camera Query:NiagaraDataInterfaceCamera=DataInterface())
        Channel=::NewEnumerator2
       )
       B=&MakeFloatFromVector(Vector=Linked() {linkedVariable {name=Particles.Position, type=NiagaraPosition}}, Channel=::NewEnumerator2)
      )
     )
    )
    Orientation Method=::NewEnumerator2
    Override Quat=false
    TransformFacingOrigin=false
    TransformFacingPosition=false
    TransformUp=false
   )
   &AlignSpriteToMeshOrientation(Mesh Orientation Relative Sprite Alignment Vector=<0 0 1>, SetAlignment=true, SetFacing=true) {module=AlignSpriteToMeshOrientation001}
   /SetVariables(
    Particles.BurstFlareColor:LinearColor=&ColorFromCurve(ColorCurve=DataInterface() {propertyValues {
       AlphaCurve {keys [<at 0 0>,<at 0.1853 1 RCIM_Cubic>,<at 1 0>]}
       BlueCurve {keys [<at 0.065 5 RCIM_Cubic>,<at 0.9925 1 RCIM_Cubic>]}
       GreenCurve {keys [<at 0.065 4.7611 RCIM_Cubic>,<at 0.9925 0.807 RCIM_Cubic>]}
       RedCurve {keys [<at 0.065 0.9423 RCIM_Cubic>,<at 0.9925 0.1948 RCIM_Cubic>]}
      }})
    Particles.FlareHoloBaseCol:LinearColor=&MultiplyLinearColor(
     Linear Color=$Particles.Initial.FlareHoloBaseCol
     Scale Factor=&MakeVector4(
      X=1
      Y=1
      Z=1
      W=&Multiply_Float(
       A=&FloatFromCurve(
        FloatCurve=DataInterface() {propertyValues {Curve {keys [
            {arriveTangent=1.2346, interpMode=RCIM_Cubic, time=0, value=1}
            {arriveTangent=-0.224, interpMode=RCIM_Cubic, leaveTangent=-0.224, tangentMode=RCTM_User, time=0.5254, value=1}
            <at 0.9192 0 RCIM_Cubic>
           ]}}}
        CurveIndex=&Abs_Float(Float:NiagaraFloat=&DotProduct(
          Vector A=&ReturnCameraVector(
           Camera Vectors=::NewEnumerator4
           Coordinate Space=::Simulation
           Origin Position=$Particles.Position
           CameraQuery=DataInterface() {propertyValues {bRequireCurrentFrameData=false}}
          )
          Vector B=<0 0 1>
         ))
       )
       B=&FloatFromCurve(FloatCurve=DataInterface() {propertyValues {Curve {
           defaultValue=1
           keys [
            {interpMode=RCIM_Cubic, tangentMode=RCTM_User, time=0, value=0}
            {interpMode=RCIM_Cubic, tangentMode=RCTM_User, time=0.0376, value=1}
            {arriveTangent=-1.073, interpMode=RCIM_Cubic, leaveTangent=-1.073, time=0.9498, value=1}
            {tangentMode=RCTM_User, time=0.9696, value=0}
           ]
          }}})
      )
     )
    )
    Particles.Lifetime=&Divide_Float(
     A=$Particles.Initial.Lifetime
     B=&MergeHelper_FloatSwitch(0=1, 1=5, 2=25, 3=?, Index=$Particles.Initial.VisibilityTag, Num=::NewEnumerator1)
    )
    Particles.LongHoloColor:LinearColor=&MultiplyLinearColor(
     Linear Color=$Particles.Initial.LongHoloColor
     Scale Factor=&MakeVector4(
      X=1
      Y=1
      Z=1
      W=&Multiply_Float(
       A=&FloatFromCurve(
        FloatCurve=DataInterface() {propertyValues {Curve {keys [
            <at 0 0 RCIM_Cubic>
            {arriveTangent=1.1924, interpMode=RCIM_Cubic, leaveTangent=1.1924, time=0.1307, value=1}
            {arriveTangent=-1.1503, interpMode=RCIM_Cubic, leaveTangent=-1.1503, time=0.8387, value=1}
            <at 1 0 RCIM_Cubic>
           ]}}}
        CurveIndex=$Emitter.NormalizedLoopAge
       )
       B=&FloatFromCurve(
        FloatCurve=DataInterface() {propertyValues {Curve {keys [
            {arriveTangent=1.2346, interpMode=RCIM_Cubic, time=0, value=1}
            {arriveTangent=-0.224, interpMode=RCIM_Cubic, leaveTangent=-0.224, tangentMode=RCTM_User, time=0.5254, value=1}
            <at 0.9192 0 RCIM_Cubic>
           ]}}}
        CurveIndex=&Abs_Float(Float:NiagaraFloat=&DotProduct(
          Vector A=&ReturnCameraVector(
           Camera Vectors=::NewEnumerator4
           Coordinate Space=::Simulation
           Origin Position=$Particles.Position
           CameraQuery=DataInterface() {propertyValues {bRequireCurrentFrameData=false}}
          )
          Vector B=<0 0 1>
         ))
       )
      )
     )
    )
    Particles.LongHoloSize:Vector2f=&Multiply_Vector2D(
     A=&Multiply_Vector2D(
      A=$Particles.Initial.LongHoloSize
      B=&Vector2DFromCurve(
       Vector2Curve=DataInterface() {propertyValues {
         XCurve {keys [
           {tangentMode=RCTM_User, time=0, value=0}
           <at 0.0284 0.1731>
           <at 0.0494 1.338>
           <at 0.2162 1>
           {arriveTangent=0.7544, interpMode=RCIM_Cubic, leaveTangent=0.7544, time=1.0015, value=1}
           {arriveTangent=-20.6173, interpMode=RCIM_Cubic, leaveTangent=-20.6173, time=1.0209, value=1.6071}
           <at 1.05 0 RCIM_Cubic>
          ]}
         YCurve {keys [{tangentMode=RCTM_User, time=0, value=0}, <at 0.0372 1.4751>, <at 0.0669 0.8642>, <at 0.1575 1>, <at 0.9 1>]}
        }}
       CurveIndex=$Emitter.NormalizedLoopAge
       ScaleCurve=<1 0.5>
      )
     )
     B=&Vector2DFromFloat(Value=&FloatFromCurve(FloatCurve=DataInterface() {propertyValues {Curve {
          defaultValue=1
          keys [
           {interpMode=RCIM_Cubic, tangentMode=RCTM_User, time=0, value=0.5}
           {interpMode=RCIM_Cubic, tangentMode=RCTM_User, time=0.5, value=1}
           {tangentMode=RCTM_User, time=1, value=0.5}
          ]
         }}}))
    )
   ) {isSetParameters=true}
   &SubUVAnimation(
    Direct Frame Index:NiagaraFloat=&RandomRangeFloat(Minimum=0, Maximum=4)
    Output Alternate Renderer Binding=true
    "Sprite Renderer "=DataInterface() {propertyValues {SpriteRenderer=&self:Rick_0.NiagaraSpriteRendererProperties_1}}
    SubUV Animation Mode:ENiagara_SubUVLookupModeV2=::NewEnumerator5
   ) {module=SubUVAnimation001}
  ]
  rendererClasses [&NiagaraSpriteRendererProperties]
  renderers [
   {rendererClass=&NiagaraSpriteRendererProperties}
   {rendererClass=&NiagaraSpriteRendererProperties, rendererIndex=1}
   {rendererClass=&NiagaraSpriteRendererProperties, rendererIndex=2}
   {rendererClass=&NiagaraSpriteRendererProperties, rendererIndex=3}
   {rendererClass=&NiagaraSpriteRendererProperties, rendererIndex=4}
  ]
 }
 RickTopBottomCover {
  emitterUpdateScript [
   &EmitterState(Inactive Response=::NewEnumerator1, Life Cycle Mode=::NewEnumerator1, Loop Behavior=::NewEnumerator1, Loop Duration=3.7, UseLoopDelay=false)
   &SpawnBurst_Instantaneous(Use Spawn Probability=false, UseLoopCountLimit=false)
  ]
  particleSpawnScript [
   &InitializeParticle(
    Color=<rgba 0.4522 1 0.9207 0.7>
    Lifetime=3.7
    Position Offset=<0 0 110>
    Sprite Size=<75 20>
    Sprite Size Mode=::NewEnumerator4
    UsePositionOffset=true
    Write Material Random=true
   )
   &SpriteFacingAndAlignment(Sprite Alignment=<0 0 1>, Sprite Facing=<0 0 1>, WriteAlignment=true, WriteFacing=true)
  ]
  particleUpdateScript [
   &SolveForcesAndVelocity(Clamp Velocity=false, Limit Acceleration=false)
   &ParticleState()
   &ScaleSpriteSize(
    Scale Factor=&MakeVector2D(
     X=1
     Y=&RemapRange(
      Input Value To Remap=&Abs_Float(Float:NiagaraFloat=&DotProduct(
        Vector A=&ReturnCameraVector(
         Camera Vectors=::NewEnumerator4
         Coordinate Space=::Simulation
         Origin Position=$Particles.Position
         CameraQuery=DataInterface() {propertyValues {bRequireCurrentFrameData=false}}
        )
        Vector B=$Particles.Initial.SpriteAlignment
       ))
      Input Min=0
      Input Max=1
      Output Min=30
      Output Max=1
     )
    )
    Scale Sprite Size Mode=::NewEnumerator2
   )
   /SetVariables(Particles.CameraZ:NiagaraFloat=&RemapRange(
     Input Value To Remap=&Abs_Float(Float:NiagaraFloat=&MakeFloatFromVector(
       Vector=&Subtract_Vector(
        A=Linked() {linkedVariable {name=Particles.Position, type=NiagaraPosition}}
        B=&ReturnCameraPosition(Coordinate Space:ENiagaraCoordinateSpace=::World, Camera Query:NiagaraDataInterfaceCamera=DataInterface())
       )
       Channel=::NewEnumerator2
      ))
     Input Min=0
     Input Max=360
     Output Min=0
     Output Max=1
    )) {isSetParameters=true}
   &ScaleColor(
    Scale Alpha=&FloatFromCurve(FloatCurve=DataInterface() {propertyValues {Curve {keys [<at 0.4 0>,<at 0.6 1>]}}}, CurveIndex=$Particles.CameraZ)
    ScaleA=true
    ScaleRGB=true
   ) {enabled=false}
   &ScaleColor(
    Scale Alpha=&FloatFromCurve(
     FloatCurve=DataInterface() {propertyValues {Curve {
        defaultValue=1
        keys [
         {
          arriveTangent=-0.182
          arriveTangentWeight=0.7464
          interpMode=RCIM_Cubic
          leaveTangent=-0.182
          leaveTangentWeight=0.7464
          tangentMode=RCTM_User
          time=0
          value=1
         }
         {
          arriveTangent=-0.3907
          arriveTangentWeight=0.2669
          interpMode=RCIM_Cubic
          leaveTangent=-0.3907
          leaveTangentWeight=0.2669
          tangentMode=RCTM_User
          time=0.7414
          value=0
         }
        ]
       }}}
     CurveIndex=&Abs_Float(Float:NiagaraFloat=&DotProduct(
       Vector A=&ReturnCameraVector(
        Camera Vectors=::NewEnumerator4
        Coordinate Space=::Simulation
        Origin Position=$Particles.Position
        CameraQuery=DataInterface() {propertyValues {bRequireCurrentFrameData=false}}
       )
       Vector B=$Particles.SpriteAlignment
      ))
    )
    ScaleA=true
    ScaleRGB=false
   ) {module=ScaleColor001}
   &UpdateMeshOrientation(
    Facing Interpolation=false
    Facing Position:NiagaraPosition=&Subtract_Vector(
     A=&ReturnCameraPosition(Coordinate Space:ENiagaraCoordinateSpace=::Simulation, Camera Query:NiagaraDataInterfaceCamera=DataInterface())
     B=&MakeVector(
      X=0
      Y=0
      Z=&Subtract_Float(
       A=&MakeFloatFromVector(
        Vector=&ReturnCameraPosition(Coordinate Space:ENiagaraCoordinateSpace=::Simulation, Camera Query:NiagaraDataInterfaceCamera=DataInterface())
        Channel=::NewEnumerator2
       )
       B=&MakeFloatFromVector(Vector=Linked() {linkedVariable {name=Particles.Position, type=NiagaraPosition}}, Channel=::NewEnumerator2)
      )
     )
    )
    Orientation Method=::NewEnumerator2
    Override Quat=false
    TransformFacingOrigin=false
    TransformFacingPosition=false
    TransformUp=false
   )
   &AlignSpriteToMeshOrientation(
    Mesh Orientation Relative Sprite Alignment Vector=<1 0 0>
    Mesh Orientation Relative Sprite Facing Vector:Vector3f=<0 0 1>
    SetAlignment=true
    SetFacing=true
   ) {module=AlignSpriteToMeshOrientation001}
   &ScaleColor(
    Scale Alpha=&FloatFromCurve(FloatCurve=DataInterface() {propertyValues {Curve {
        defaultValue=1
        keys [
         {interpMode=RCIM_Cubic, tangentMode=RCTM_User, time=0, value=0}
         {interpMode=RCIM_Cubic, tangentMode=RCTM_User, time=0.2, value=1}
         {arriveTangent=-0.0951, interpMode=RCIM_Cubic, leaveTangent=-0.0951, tangentMode=RCTM_User, time=0.8, value=1}
         {tangentMode=RCTM_User, time=1, value=0}
        ]
       }}})
    ScaleA=true
    ScaleRGB=true
   ) {module=ScaleColor002}
  ]
  rendererClasses [&NiagaraSpriteRendererProperties]
  renderers [{rendererClass=&NiagaraSpriteRendererProperties}]
 }
}
summary {
 systemName=NS_Emote_ClimbSpill_PropSpawn
 emitters [
  {emitterName=Confetti, rendererClasses [&NiagaraSpriteRendererProperties]}
  {emitterName=Rick, rendererClasses [&NiagaraSpriteRendererProperties]}
  {emitterName=RickTopBottomCover, rendererClasses [&NiagaraSpriteRendererProperties]}
  {emitterName=Beam_LWE, rendererClasses []}
  {emitterName=GlitchBox, rendererClasses []}
  {emitterName=Flash001, rendererClasses [&NiagaraSpriteRendererProperties, &NiagaraMeshRendererProperties]}
  {emitterName=GlitchBoxBottom, rendererClasses []}
 ]
 userVariables [User.bIsFrontEndPreview {type=NiagaraBool}]
}
