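# Niagara effect for the LentilAnise emote, showing timer digits as two oppositely facing sprites with a squash-and-stretch pop-in.
# User.TimerCurve feeds the emitter timer and is remapped to drive lit and unlit digit colors between orange and red, with a blink keyed to emitter age.
# Particle-age curves drive dynamic material parameters.
# The Minimal emitter adds two warm fading lights; Smoke is disabled and Cover has no renderer.
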NiagaraSystem /BRCosmetics/Animation/Game/MainPlayer/Emotes/LentilAnise/FX/NS_Emote_LentilAnise_Digits
dependencies {
 usedDataInterfaces [
  &NiagaraDataInterfaceColorCurve
  &NiagaraDataInterfaceCurve
  &NiagaraDataInterfaceVector2DCurve
  &NiagaraDataInterfaceSkeletalMesh
 ]
 usedDynamicInputs []
 usedModules [&SystemState]
}
emitters {
 Cover {rendererClasses [], renderers []}
 Digits {
  emitterUpdateScript [
   &EmitterState(Life Cycle Mode=::NewEnumerator1, Loop Behavior=::NewEnumerator1, Loop Duration=1, UseLoopDelay=false)
   &SpawnBurst_Instantaneous(Spawn Count=2, Spawn Time=0.1, Use Spawn Probability=false, UseLoopCountLimit=false)
   /SetVariables(Emitter.RemappedCurve:NiagaraFloat=&RemapRange(Input Value To Remap=$User.TimerCurve, Input Min=0, Input Max=5.35, Output Min=1, Output Max=0)) {isSetParameters=true}
   /SetVariables(
    Emitter.DigitColor:LinearColor=&ColorFromCurve(
     ColorCurve=DataInterface() {propertyValues {
       AlphaCurve {keys [<at 0 1 RCIM_Cubic>]}
       BlueCurve {keys [<at 0.009 0>,<at 0.2245 0 RCIM_Cubic>,<at 0.8707 0 RCIM_Cubic>]}
       GreenCurve {keys [
         <at 0.009 0.5121>
         {arriveTangent=-0.5389, interpMode=RCIM_Cubic, leaveTangent=-0.5389, time=0.2245, value=0.2488}
         <at 0.8707 0.0477 RCIM_Cubic>
        ]}
       RedCurve {keys [<at 0.009 1>,<at 0.2245 1 RCIM_Cubic>,<at 0.8707 1 RCIM_Cubic>]}
      }}
     CurveIndex=$Emitter.RemappedCurve
    )
    Emitter.DigitColorOff:LinearColor=&ColorFromCurve(
     ColorCurve=DataInterface() {propertyValues {
       AlphaCurve {keys [<at 0 1 RCIM_Cubic>]}
       BlueCurve {keys [<at 0.009 0>,<at 0.2245 0 RCIM_Cubic>,<at 0.8707 0 RCIM_Cubic>]}
       GreenCurve {keys [
         <at 0.009 0.3073>
         {arriveTangent=-0.3233, interpMode=RCIM_Cubic, leaveTangent=-0.3233, time=0.2245, value=0.1493}
         <at 0.8707 0.0286 RCIM_Cubic>
        ]}
       RedCurve {keys [<at 0.009 0.6>,<at 0.2245 0.6 RCIM_Cubic>,<at 0.8707 0.6 RCIM_Cubic>]}
      }}
     CurveIndex=$Emitter.RemappedCurve
    )
    Emitter.Timer:NiagaraFloat=$User.TimerCurve
   ) {isSetParameters=true}
   /SetVariables(
    Emitter.DigitColor:LinearColor=&Multiply_LinearColorByFloat(
     Linear Color=$Emitter.DigitColor
     Float=&Lerp_Float(
      A=1
      B=/Niagara/DynamicInputs/Integer/V2/Round_Float(Float:NiagaraFloat=&Frac_Float(Float=&Multiply_Float(A=$Emitter.Age, B=2)))
      Alpha=&MakeCustomFloatFromBool(Bool=&SetBoolByFloatComparison(A=$Emitter.Age, B=5.35), True Float=5, False Float=0)
     )
    )
    Emitter.DigitColorOff:LinearColor=&Multiply_LinearColorByFloat(
     Linear Color=$Emitter.DigitColorOff
     Float=&Lerp_Float(
      A=1
      B=&Round_Float(Float:NiagaraFloat=&Frac_Float(Float=&Multiply_Float(A=$Emitter.Age, B=2)))
      Alpha=&MakeCustomFloatFromBool(Bool=&SetBoolByFloatComparison(A=$Emitter.Age, B=5.35), True Float=5, False Float=0)
     )
    )
   ) {isSetParameters=true}
  ]
  particleSpawnScript [
   &PartitionParticles()
   &InitializeParticle(
    Color Mode=::NewEnumerator0
    Lifetime=0
    Position Offset=&MergeHelper_VectorSwitch(0=<0 0 0>, 1=<-43 0 0>, 2=?, 3=?, Index=$Particles.PartitionParticles.Partition, Num=::NewEnumerator0)
    Sprite Size=&Multiply_Vector2D_ByFloat(Float=80, Vector2D=<0.8 0.4>)
    Sprite Size Mode=::NewEnumerator4
    UsePositionOffset=true
    Write Material Random=true
   )
   &InitialMeshOrientation(
    Mesh Axis To Orient=<0 0 0>
    Mesh Orientation Mode=::NewEnumerator2
    Orientation Vector=&MergeHelper_VectorSwitch(0=<1 0 0>, 1=<-1 0 0>, 2=?, 3=?, Index=$Particles.PartitionParticles.Partition, Num=::NewEnumerator0)
    Rotation=<0.25 0 0>
    Use Rotation Vector=true
   )
   &ApplyInitialForces()
   &AlignSpriteToMeshOrientation(SetAlignment=true, SetFacing=true)
  ]
  particleUpdateScript [
   &ParticleState(Kill Particles When Lifetime Has Elapsed=false)
   &ScaleSpriteSize(
    Non-Uniform Curve Index:NiagaraFloat=$Particles.Age
    Non-Uniform Curve Sprite Scale=DataInterface() {propertyValues {
      XCurve {keys [{tangentMode=RCTM_User, time=0, value=0}, <at 0.0648 0.1731>, <at 0.1751 1.338>, <at 0.25 1>]}
      YCurve {keys [{tangentMode=RCTM_User, time=0, value=0}, <at 0.0878 1.4818>, <at 0.1593 0.8642>, <at 0.25 1>]}
     }}
    Scale Sprite Size Mode=::NewEnumerator3
   )
   &DynamicMaterialParameters(
    Index 0 Param 1=&FloatFromCurve(
     FloatCurve=DataInterface() {propertyValues {Curve {keys [
         {arriveTangent=-4.1758, interpMode=RCIM_Cubic, leaveTangent=-4.1758, tangentMode=RCTM_User, time=0, value=1}
         {arriveTangent=-1.1026, interpMode=RCIM_Cubic, leaveTangent=-1.1026, tangentMode=RCTM_User, time=0.35, value=0}
        ]}}}
     CurveIndex=$Particles.Age
    )
    Index 0 Param 2=&FloatFromCurve(FloatCurve=DataInterface() {propertyValues {Curve {keys [<at 0 1>,<at 1 0>]}}}, CurveIndex=$Particles.Age, Scale Curve=20)
    Module.Index0Float0Write=true
    Module.Index0Float1Write=true
    Module.Index0Float2Write=true
    Module.Index0Float3Write=true
   )
  ]
  rendererClasses [&NiagaraSpriteRendererProperties]
  renderers [{rendererClass=&NiagaraSpriteRendererProperties}]
 }
 Minimal {
  emitterUpdateScript [
   &EmitterState(Life Cycle Mode=::NewEnumerator1, Loop Behavior=::NewEnumerator1, UseLoopDelay=false)
   &SpawnBurst_Instantaneous(Spawn Count=2, Use Spawn Probability=false, UseLoopCountLimit=false)
  ]
  particleSpawnScript [
   &PartitionParticles()
   &InitializeParticle(
    Color=<rgba 1 0.246 0.1084 1>
    Lifetime=6
    Position Offset=&MergeHelper_VectorSwitch(0=<-20 35 0>, 1=<-20 -35 0>, 2=?, 3=?, Index=$Particles.PartitionParticles.Partition, Num=::NewEnumerator0)
    UsePositionOffset=true
    Write Material Random=true
   )
  ]
  particleUpdateScript [
   &ParticleState()
   &Light_Attributes(Diffuse Scale=1, Enabled=1, Falloff Exponent=1, Light Diffuse Scale=10, Light Specular Scale=10, Radius=1, Specular Scale=1)
   &ScaleColor(
    Scale Alpha=&FloatFromCurve(FloatCurve=DataInterface() {propertyValues {Curve {
        defaultValue=1
        keys [
         {interpMode=RCIM_Cubic, tangentMode=RCTM_User, time=0, value=0}
         {interpMode=RCIM_Cubic, tangentMode=RCTM_User, time=0.0297, value=1}
         {arriveTangent=-1.0306, interpMode=RCIM_Cubic, leaveTangent=-1.0306, time=0.8574, value=1}
         {tangentMode=RCTM_User, time=1, value=0}
        ]
       }}})
    ScaleA=true
    ScaleRGB=true
   )
  ]
  rendererClasses [&NiagaraLightRendererProperties]
  renderers [{rendererClass=&NiagaraLightRendererProperties, rendererIndex=0}]
 }
 Smoke {
  bEnabled=false
  emitterUpdateScript [
   &EmitterState(Life Cycle Mode=::NewEnumerator1, Loop Behavior=::NewEnumerator1, Loop Duration=0.2, UseLoopDelay=false)
   &SpawnBurst_Instantaneous(Use Spawn Probability=false, UseLoopCountLimit=false)
  ]
  particleSpawnScript [
   /SetVariables(Particles.Random=&RandomRangeFloat(Minimum=0, Maximum=1)) {isSetParameters=true}
   &CameraOffset(Camera Offset Amount=30)
   &DynamicMaterialParameters(
    Index 0 Param 2=&UniformAOrBFloat(A=0)
    Module.Param0WriteEnabled=true
    Module.Param1WriteEnabled=true
    Module.Param2WriteEnabled=true
    Module.Param3WriteEnabled=true
   ) {module=DynamicMaterialParameters001}
   &InitializeParticle(
    Color=$System.SmokeCol
    Lifetime=&RandomRangeFloat(Minimum=0.9, Maximum=0.9)
    Position Offset=<0 0 -85>
    Sprite Rotation Mode=::NewEnumerator2
    Sprite Size Mode=::NewEnumerator3
    Uniform Sprite Size=&Lerp_Float(A=110, B=110, Alpha=$Particles.Random)
    UsePositionOffset=true
    Write Material Random=true
   ) {module=InitializeParticle001}
   &ApplyInitialForces()
   &InitialMeshOrientation(Mesh Orientation Mode=::NewEnumerator2, Orientation Coordinate Space=::World, Rotation=<0 -0.25 0>, Use Rotation Vector=true)
   &SkeletalMeshLocation(
    Bone Sampling Mode=::NewEnumerator4
    MaskSpawnGroup=false
    Orientation Sampling=::NewEnumerator1
    Position Sampling=::NewEnumerator1
    Sampled Position Offset=<0 0 -60>
    Skeletal Mesh=DataInterface() {propertyValues {
      DefaultMesh=null
      ExcludeBoneName=root
      OptionalComponentClass=null
      PreviewMesh=null
      SkinningMode=SkinOnTheFly
      SoftSourceActor=null
      SourceMode=AttachParent
      UvSetIndex=0
      WholeMeshLOD=-1
      bExcludeBone=false
      bReadDeformedGeometry=true
      bRequireCurrentFrameData=true
      ComponentTags []
      FilteredBones [JNT_COG]
      FilteredSockets []
      MeshUserParameter {parameter {name=None, typeDefHandle {registeredTypeIndex=1}, varData []}}
      SamplingRegions []
     }}
   )
  ]
  particleUpdateScript [
   &ParticleState()
   &ScaleSpriteSize(Uniform Curve Sprite Scale=DataInterface() {propertyValues {Curve {keys [
        {arriveTangent=1.6815, interpMode=RCIM_Cubic, leaveTangent=1.6815, tangentMode=RCTM_User, time=0.2, value=0.5}
        {arriveTangent=0.1063, interpMode=RCIM_Cubic, leaveTangent=0.1063, tangentMode=RCTM_User, time=0.55, value=1}
       ]}}})
   &SolveForcesAndVelocity(Clamp Velocity=false, Limit Acceleration=false)
   &AlignSpriteToMeshOrientation(SetAlignment=true, SetFacing=true)
   &DynamicMaterialParameters(
    Index 0 Param 1=&FloatFromCurve(FloatCurve=DataInterface() {propertyValues {Curve {keys [
         {arriveTangent=1.7661, interpMode=RCIM_Cubic, leaveTangent=1.7661, tangentMode=RCTM_User, time=0, value=0}
         {arriveTangent=0.0224, interpMode=RCIM_Cubic, leaveTangent=0.0224, tangentMode=RCTM_User, time=1, value=1}
        ]}}})
    Index 0 Param 2=1
    Module.Param0WriteEnabled=true
    Module.Param1WriteEnabled=true
    Module.Param2WriteEnabled=true
    Module.Param3WriteEnabled=true
   )
   &ScaleColor(
    Scale Alpha=&FloatFromCurve(FloatCurve=DataInterface() {propertyValues {Curve {keys [<at 0 0>, {tangentMode=RCTM_User, time=0.75, value=1}, {tangentMode=RCTM_User, time=1, value=0}]}}})
    Scale RGB=&MakeVectorFromLinearColorRGB(Color=&ColorFromCurve(ColorCurve=DataInterface() {propertyValues {
        AlphaCurve {keys [<at 0 0>,<at 0.4487 1 RCIM_Cubic>]}
        BlueCurve {keys [
          <at 0.1811 0.6425 RCIM_Cubic>
          {arriveTangent=-0.3057, interpMode=RCIM_Cubic, leaveTangent=-0.3057, time=0.4806, value=1}
          <at 0.7837 0.4583>
         ]}
        GreenCurve {keys [
          <at 0.1811 0.6562 RCIM_Cubic>
          {arriveTangent=-0.3285, interpMode=RCIM_Cubic, leaveTangent=-0.3285, time=0.4806, value=1}
          <at 0.7837 0.4583>
         ]}
        RedCurve {keys [
          <at 0.1811 0.6388 RCIM_Cubic>
          {arriveTangent=-0.2996, interpMode=RCIM_Cubic, leaveTangent=-0.2996, time=0.4806, value=1}
          <at 0.7837 0.4583>
         ]}
       }}))
    ScaleA=true
    ScaleRGB=true
   )
  ]
  rendererClasses [&NiagaraSpriteRendererProperties]
  renderers [{rendererClass=&NiagaraSpriteRendererProperties}]
 }
}
summary {
 systemName=NS_Emote_LentilAnise_Digits
 emitters [
  {emitterName=Digits, rendererClasses [&NiagaraSpriteRendererProperties]}
  {bEnabled=false, emitterName=Smoke, rendererClasses [&NiagaraSpriteRendererProperties]}
  {emitterName=Cover, rendererClasses []}
  {emitterName=Minimal, rendererClasses [&NiagaraLightRendererProperties]}
 ]
 userVariables [
  User.CountdownTime {type=NiagaraFloat, defaultValue=10, description=""}
  User.CountUp {type=NiagaraBool, defaultValue=true}
  User.TimerCurve {type=NiagaraFloat, defaultValue=0, description=""}
  User.TimerDelay {type=NiagaraFloat, defaultValue=1.32, description=""}
  User.TimerStart {type=NiagaraFloat, defaultValue=0, description=""}
 ]
}
