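# Niagara system for the LentilAnise emote's "PortalB": a short-lived green portal with a light that opens and closes.
# Flash bursts bracket the portal, joined by yellow electric ring sprites, a delayed downward ribbon trail and curl-noise sparks.
# Portal glow stretches by camera angle to the vertical axis, apparently replacing the two disabled Portal_Glow emitters.
# Front-end modifiers adjust trail alpha and spark speed and size.
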NiagaraSystem /BRCosmetics/Animation/Game/MainPlayer/Emotes/LentilAnise/FX/NS_Emote_LentilAnise_PortalB
dependencies {
 usedDataInterfaces [&NiagaraDataInterfaceCurve, &NiagaraDataInterfaceCamera, &NiagaraDataInterfaceColorCurve]
 usedDynamicInputs []
 usedModules [&SystemState]
}
emitters {
 Air_Trail_Up {
  emitterUpdateScript [
   &EmitterState(
    Inactive Response=::NewEnumerator1
    Life Cycle Mode=::NewEnumerator1
    Loop Behavior=::NewEnumerator1
    Loop Delay=0.6
    Loop Duration=0.2
    UseLoopDelay=true
   )
   &SpawnRate(SpawnRate=400, Use Spawn Probability=false)
  ]
  particleSpawnScript [
   &InitializeParticle(
    Color=<rgba 5 5 5 1>
    Lifetime=0.3
    Position Offset=<0 5 0>
    Position Offset Coordinate Space=::Local
    Ribbon Width=70
    Ribbon Width Mode=::NewEnumerator1
    UsePositionOffset=true
    Write Material Random=true
   )
   &AddVelocity(Velocity=<0 0 -2500>, Velocity Speed Scale=0.8)
  ]
  particleUpdateScript [
   &ParticleState()
   &SolveForcesAndVelocity(Clamp Velocity=false, Limit Acceleration=false)
   &ScaleColor(
    Scale Alpha=&FloatFromCurve(
     FloatCurve=DataInterface() {propertyValues {Curve {keys [
         {arriveTangent=-2.5607, interpMode=RCIM_Cubic, leaveTangent=-2.5607, leaveTangentWeight=1.4087, tangentMode=RCTM_User, time=0, value=1}
         {interpMode=RCIM_Cubic, leaveTangent=0.45, time=1, value=0}
        ]}}}
     CurveIndex=$Emitter.NormalizedLoopAge
     Scale Curve=&FN_FrontEndFloatModifier(Gameplay=4, Front End=3)
    )
    Scale RGB=&MakeVectorFromLinearColorRGB(Color=&ColorFromCurve(ColorCurve=DataInterface() {propertyValues {
        AlphaCurve {keys []}
        BlueCurve {keys [
          <at 0 0.3802>
          {arriveTangent=-0.9804, interpMode=RCIM_Cubic, leaveTangent=-0.9804, time=0.0299, value=0.8646}
          <at 0.17 0.2135>
          <at 0.35 0.3125 RCIM_Cubic>
         ]}
        GreenCurve {keys [
          <at 0 0.3802>
          {arriveTangent=-0.9804, interpMode=RCIM_Cubic, leaveTangent=-0.9804, time=0.0299, value=0.8646}
          <at 0.17 0.2135>
          <at 0.35 0.3125 RCIM_Cubic>
         ]}
        RedCurve {keys [
          <at 0 0.3802>
          {arriveTangent=-0.9804, interpMode=RCIM_Cubic, leaveTangent=-0.9804, time=0.0299, value=0.8646}
          <at 0.17 0.2135>
          <at 0.35 0.3125 RCIM_Cubic>
         ]}
       }}))
    ScaleA=true
    ScaleRGB=true
   )
  ]
  rendererClasses [&NiagaraRibbonRendererProperties]
  renderers [{rendererClass=&NiagaraRibbonRendererProperties}]
 }
 ElectricSprites {
  emitterUpdateScript [
   &EmitterState()
   &SpawnRate(
    SpawnRate=&FloatFromCurve(
     FloatCurve=DataInterface() {propertyValues {Curve {keys [<at 0 1>,<at 0.2867 1>,<at 0.3196 0>]}}}
     CurveIndex=$Emitter.NormalizedLoopAge
     Scale Curve=40
    )
    Use Spawn Probability=false
   )
  ]
  particleSpawnScript [
   &FN_SetSpawnGroup(Number of Spawn Groups=11)
   &InitializeParticle(
    Color=<rgba 10 8.7282 1.3948 2>
    Lifetime Max=0.25
    Lifetime Min=0.1
    Lifetime Mode=::NewEnumerator1
    Position Offset=<0 -10 0>
    Position Offset Coordinate Space=::Local
    Sprite Rotation Mode=::NewEnumerator2
    Sprite Size=&UniformRangedVector2D(Minimum=<200 200>, Maximum=<400 400>)
    Sprite Size Mode=::NewEnumerator4
    Sprite UV Mode=::NewEnumerator1
    UsePositionOffset=true
    Write Material Random=true
   ) {module=InitializeParticle001}
   &ShapeLocation(
    MaskSpawnGroup=false
    Override Seed=false
    Ring Radius=&FloatFromCurve(FloatCurve=DataInterface() {propertyValues {Curve {keys [<at 0 0>,<at 0.1803 1>]}}}, CurveIndex=$Emitter.NormalizedLoopAge, Scale Curve=50)
    Shape Primitive=::NewEnumerator4
    UseOwnerScale=true
   )
   &DynamicMaterialParameters(
    Index 0 Param 1=$Particles.MaterialRandom
    Index 0 Param 2=&Lerp_Float(A=3, B=5, Alpha=$Particles.MaterialRandom)
    Module.Param0WriteEnabled=true
    Module.Param1WriteEnabled=true
    Module.Param2WriteEnabled=false
    Module.Param3WriteEnabled=false
   )
   &CameraOffset(Camera Offset Amount=&RandomRangeFloat(Minimum=-10, Maximum=10))
   &SubUVAnimation(End Frame=3, SubUV Animation Mode:ENiagara_SubUVLookupMode=::NewEnumerator2) {module=SubUVAnimation001}
  ]
  particleUpdateScript [
   &ParticleState()
   &DynamicMaterialParameters(
    Index 0 Param 3=&FloatFromCurve(FloatCurve=DataInterface() {propertyValues {Curve {keys [<at 0.2 0 RCIM_Cubic>, {interpMode=RCIM_Cubic, leaveTangent=1.5819, time=1, value=1}]}}})
    Module.Param0WriteEnabled=false
    Module.Param1WriteEnabled=false
    Module.Param2WriteEnabled=true
    Module.Param3WriteEnabled=false
   ) {module=DynamicMaterialParameters001}
   &ScaleSpriteSize(Scale Factor=&Vector2DFromFloat(Value=&FloatFromCurve(
      FloatCurve=DataInterface() {propertyValues {Curve {keys [{arriveTangent=0.3147, interpMode=RCIM_Cubic, leaveTangent=0.3147, tangentMode=RCTM_User, time=0, value=0.8}, <at 1 1>]}}}
      Scale Curve=&RandomRangeFloat(Minimum=0.1, Maximum=0.1)
     )))
   &ScaleColor(
    Scale Alpha=&FNTimeOfDayScalar(Night Scale=0.1, Morning Scale=0.1, Evening Scale=0.1)
    ScaleA=true
    ScaleRGB=true
   )
  ]
  rendererClasses [&NiagaraSpriteRendererProperties]
  renderers [{rendererClass=&NiagaraSpriteRendererProperties}]
 }
 Flash001 {
  emitterUpdateScript [
   &EmitterState()
   &SpawnBurst_Instantaneous(Spawn Count=4, Use Spawn Probability=false)
   &SpawnBurst_Instantaneous(Spawn Count=4, Spawn Time=1.2, Use Spawn Probability=false, UseLoopCountLimit=false) {module=SpawnBurst_Instantaneous001}
  ]
  particleSpawnScript [
   &PartitionParticles(Number Of Partitions=4)
   &InitializeParticle(
    Color=<rgba 0.7241 1 0.167 0.5>
    Lifetime=&MakeCustomFloatFromBool(Bool=&SetBoolByFloatComparison(A=$Emitter.Age, B=0.8), True Float=0.25, False Float=0.35)
    Sprite Size Mode=::NewEnumerator3
    Uniform Sprite Size=&Multiply_Float(A=100, B=1)
    UsePositionOffset=true
    Write Material Random=true
   )
   &ShapeLocation(MaskSpawnGroup=false, Override Seed=false, Sphere Radius=0, UseOwnerScale=true)
   &CameraOffset(Camera Offset Amount=-10)
   /SetVariables(Particles.SizeWideFlare:Vector2f=&Multiply_Vector2D_ByFloat(Float=&Multiply_Float(A=1, B=1), Vector2D=<225 40>)) {isSetParameters=true}
   &InitialMeshOrientation(Mesh Orientation Mode=::NewEnumerator2, Orientation Vector=<0 1 0>, Rotation=<0.25 0 0>, Use Rotation Vector=true)
   &AlignSpriteToMeshOrientation(SetAlignment=true, SetFacing=true) {enabled=false}
  ]
  particleUpdateScript [
   &ParticleState()
   &SolveForcesAndVelocity(Clamp Velocity=false, Limit Acceleration=false)
   &ScaleColor(
    Scale Alpha=&FloatFromCurve(FloatCurve=DataInterface() {propertyValues {Curve {keys [
         <at 0.1 0 RCIM_Cubic>
         {interpMode=RCIM_Cubic, leaveTangent=-2.0997, tangentMode=RCTM_Break, time=0.2521, value=1}
         {arriveTangentWeight=0.7369, interpMode=RCIM_Cubic, tangentMode=RCTM_Break, time=1, value=1}
        ]}}})
    Scale RGB=&VectorFromFloat(Value=&FloatFromCurve(
      FloatCurve=DataInterface() {propertyValues {Curve {keys [
          <at 0.1 0 RCIM_Cubic>
          {interpMode=RCIM_Cubic, leaveTangent=-2.0997, tangentMode=RCTM_Break, time=0.2521, value=1}
          {arriveTangentWeight=0.7369, interpMode=RCIM_Cubic, tangentMode=RCTM_Break, time=0.6, value=1}
         ]}}}
      Scale Curve=5
     ))
    ScaleA=true
    ScaleRGB=true
   )
   &ScaleSpriteSize(
    Uniform Curve Scale=&MakeCustomFloatFromBool(Bool=&SetBoolByFloatComparison(A=$Emitter.Age, B=0.8), True Float=0.4, False Float=1)
    Uniform Curve Sprite Scale=DataInterface() {propertyValues {Curve {keys [
        <at 0.3 0 RCIM_Cubic>
        {interpMode=RCIM_Cubic, tangentMode=RCTM_User, time=0.4, value=1}
        {interpMode=RCIM_Cubic, tangentMode=RCTM_User, time=1, value=0}
       ]}}}
   )
   &FN_Sprite_Attributes(
    Color=&MakeLinearColorFromVectorAndFloat(
     "Vector (RGB)"=&MakeVectorFromLinearColorRGB(Color=<rgba 0.0658 0.3391 0 1>)
     "Float (Alpha)"=&Multiply_Float(
      A=1
      B=&FloatFromCurve(FloatCurve=DataInterface() {propertyValues {Curve {keys [{interpMode=RCIM_Cubic, tangentMode=RCTM_User, time=0, value=1}, {interpMode=RCIM_Cubic, tangentMode=RCTM_User, time=1, value=0.5}]}}})
     )
    )
    Create Color Attribute=true
    Write to Module Namespace=true
   )
   /SetVariables(Particles.SizeWideFlare:Vector2f=&Multiply_Vector2D(
     A=$Particles.Initial.SizeWideFlare
     B=&MakeVector2D(
      X=&FloatFromCurve(FloatCurve=DataInterface() {propertyValues {Curve {keys [<at 0.1 0>,<at 0.2 0.4>,<at 0.3 1>,<at 1 0>]}}})
      Y=&FloatFromCurve(FloatCurve=DataInterface() {propertyValues {Curve {keys [<at 0.2 1>,<at 0.5 0.1>]}}})
     )
    )) {isSetParameters=true}
  ]
  rendererClasses [&NiagaraSpriteRendererProperties, &NiagaraMeshRendererProperties]
  renderers [
   {rendererClass=&NiagaraSpriteRendererProperties}
   {rendererClass=&NiagaraSpriteRendererProperties, rendererIndex=1}
   {rendererClass=&NiagaraMeshRendererProperties, rendererIndex=2}
  ]
 }
 Portal {
  emitterUpdateScript [
   &EmitterState()
   &SpawnBurst_Instantaneous(Spawn Time=0.05, Use Spawn Probability=false, UseLoopCountLimit=false)
   &SpawnBurst_Instantaneous(Spawn Group=1, Spawn Time=0.05, Use Spawn Probability=false, UseLoopCountLimit=false) {module=SpawnBurst_Instantaneous001}
  ]
  particleSpawnScript [
   &InitializeParticle(Lifetime=1.3, Sprite Size Mode=::NewEnumerator3, Uniform Sprite Size=130, UsePositionOffset=true, Write Material Random=true)
   /SetVariables(Particles.VisibilityTag=$Emitter.SpawnGroup) {isSetParameters=true}
   &InitialMeshOrientation(Mesh Axis To Orient=<0 0 0>, Mesh Orientation Mode=::NewEnumerator2, Rotation=<-0.195 0 0>, Use Rotation Vector=true)
   &SpriteFacingAndAlignment(Sprite Facing=<0 0 1>, WriteAlignment=false, WriteFacing=true)
   /SetVariables(
    Particles.GlowColor=<rgba 1.199 2 0.3599 0.7>
    Particles.GlowSize=<215 20>
    Particles.LightPos:NiagaraPosition=&SubtractVectorFromPosition(Position=$Engine.Emitter.SimulationPosition, Vector=<0 0 0>)
   ) {isSetParameters=true}
  ]
  particleUpdateScript [
   &ParticleState()
   &ScaleSpriteSize(Uniform Curve Sprite Scale=DataInterface() {propertyValues {Curve {
       defaultValue=0
       keys [
        {interpMode=RCIM_Cubic, tangentMode=RCTM_User, time=-0.0221, value=0}
        {arriveTangent=-1.7828, arriveTangentWeight=0.1, interpMode=RCIM_Cubic, leaveTangent=-1.7828, tangentMode=RCTM_User, time=0.2779, value=1}
        {
         arriveTangent=0.0159
         arriveTangentWeight=0.2006
         interpMode=RCIM_Cubic
         leaveTangent=0.0159
         tangentMode=RCTM_User
         time=0.8796
         value=0.988
        }
        <at 0.9779 0>
       ]
      }}})
   &ScaleMeshSize(Scale Factor=&VectorFromFloat(Value=&FloatFromCurve(
      FloatCurve=DataInterface() {propertyValues {Curve {
         defaultValue=1
         keys [
          {interpMode=RCIM_Cubic, tangentMode=RCTM_User, time=0, value=0}
          {arriveTangent=-1.7828, interpMode=RCIM_Cubic, leaveTangent=-1.7828, tangentMode=RCTM_User, time=0.3, value=1}
          {arriveTangent=0.0159, interpMode=RCIM_Cubic, leaveTangent=0.0159, tangentMode=RCTM_User, time=0.9018, value=0.988}
          {tangentMode=RCTM_User, time=1, value=0}
         ]
        }}}
      Scale Curve=$User.PortalScale
     )))
   /SetVariables(
    Particles.GlowCamOffset:NiagaraFloat=&FloatFromCurve(
     FloatCurve=DataInterface() {propertyValues {Curve {
        defaultValue=1
        keys [
         {
          arriveTangent=-0.182
          arriveTangentWeight=0.7464
          interpMode=RCIM_Cubic
          leaveTangent=-0.182
          leaveTangentWeight=0.7464
          tangentMode=RCTM_User
          time=0.85
          value=-50
         }
         {
          arriveTangent=-0.3907
          arriveTangentWeight=0.2669
          interpMode=RCIM_Cubic
          leaveTangent=-0.3907
          leaveTangentWeight=0.2669
          tangentMode=RCTM_User
          time=1
          value=-50
         }
        ]
       }}}
     CurveIndex=&Abs_Float(Float:NiagaraFloat=&DotProduct(
       Vector A=&ReturnCameraVector(
        Camera Vectors=::NewEnumerator4
        Coordinate Space=::Simulation
        Origin Position=$Particles.Position
        CameraQuery=DataInterface() {propertyValues {bRequireCurrentFrameData=false}}
       )
       Vector B=<0 0 1>
      ))
    )
    Particles.GlowSize=&Multiply_Vector2D(
     A=$Particles.Initial.GlowSize
     B=&Multiply_Vector2D_ByFloat(
      Float=&FloatFromCurve(FloatCurve=DataInterface() {propertyValues {Curve {
          defaultValue=0
          keys [
           {interpMode=RCIM_Cubic, tangentMode=RCTM_User, time=-0.0221, value=0}
           {arriveTangent=-1.7828, arriveTangentWeight=0.1, interpMode=RCIM_Cubic, leaveTangent=-1.7828, tangentMode=RCTM_User, time=0.2779, value=1}
           {
            arriveTangent=0.0159
            arriveTangentWeight=0.2006
            interpMode=RCIM_Cubic
            leaveTangent=0.0159
            tangentMode=RCTM_User
            time=0.8796
            value=0.988
           }
           <at 0.9779 0>
          ]
         }}})
      Vector2D=&MakeVector2D(
       X=1
       Y=&FloatFromCurve(
        FloatCurve=DataInterface() {propertyValues {Curve {
           defaultValue=1
           keys [
            {
             arriveTangent=-0.182
             arriveTangentWeight=0.7464
             interpMode=RCIM_Cubic
             leaveTangent=-0.182
             leaveTangentWeight=0.7464
             tangentMode=RCTM_User
             time=0
             value=1
            }
            {
             arriveTangent=-0.3907
             arriveTangentWeight=0.2669
             interpMode=RCIM_Cubic
             leaveTangent=-0.3907
             leaveTangentWeight=0.2669
             tangentMode=RCTM_User
             time=1
             value=9
            }
           ]
          }}}
        CurveIndex=&Abs_Float(Float:NiagaraFloat=&DotProduct(
          Vector A=&ReturnCameraVector(
           Camera Vectors=::NewEnumerator4
           Coordinate Space=::Simulation
           Origin Position=$Particles.Position
           CameraQuery=DataInterface() {propertyValues {bRequireCurrentFrameData=false}}
          )
          Vector B=<0 0 1>
         ))
       )
      )
     )
    )
   ) {isSetParameters=true}
   &Light_Attributes(
    Diffuse Scale=1
    Enabled=1
    Falloff Exponent=1
    Light Diffuse Scale=10
    Light Radius=120
    Light Specular Scale=10
    Radius=1
    Specular Scale=1
   )
   /SetVariables(Particles.LightCol:LinearColor=&MultiplyLinearColor(
     Linear Color=$Particles.Initial.Color
     Scale Factor=&MakeVector4(
      X=1
      Y=1
      Z=1
      W=&FloatFromCurve(FloatCurve=DataInterface() {propertyValues {Curve {
          defaultValue=1
          keys [
           {interpMode=RCIM_Cubic, tangentMode=RCTM_User, time=0, value=0}
           {arriveTangent=1.2984, interpMode=RCIM_Cubic, leaveTangent=1.2984, time=0.1219, value=1}
           {interpMode=RCIM_Cubic, tangentMode=RCTM_User, time=0.7702, value=1}
           {tangentMode=RCTM_User, time=1, value=0}
          ]
         }}})
     )
    )) {isSetParameters=true}
  ]
  rendererClasses [&NiagaraSpriteRendererProperties, &NiagaraLightRendererProperties]
  renderers [
   {rendererClass=&NiagaraSpriteRendererProperties}
   {rendererClass=&NiagaraSpriteRendererProperties, rendererIndex=1}
   {rendererClass=&NiagaraLightRendererProperties, rendererIndex=2}
  ]
 }
 Portal_Glow {
  bEnabled=false
  emitterUpdateScript [
   &EmitterState()
   &SpawnBurst_Instantaneous(Spawn Time=0.05, Use Spawn Probability=false, UseLoopCountLimit=false)
  ]
  particleSpawnScript [&InitializeParticle(
    Color=<rgba 1.199 2 0.3599 0.7>
    Lifetime=1.3
    Sprite Size=<200 20>
    Sprite Size Mode=::NewEnumerator4
    UsePositionOffset=true
    Write Material Random=true
   )]
  particleUpdateScript [
   &SolveForcesAndVelocity(Clamp Velocity=false, Limit Acceleration=false)
   &ParticleState()
   &ScaleSpriteSize(
    Scale Factor=&Multiply_Vector2D_ByFloat(
     Float=&FloatFromCurve(FloatCurve=DataInterface() {propertyValues {Curve {
         defaultValue=0
         keys [
          {interpMode=RCIM_Cubic, tangentMode=RCTM_User, time=-0.0221, value=0}
          {arriveTangent=-1.7828, arriveTangentWeight=0.1, interpMode=RCIM_Cubic, leaveTangent=-1.7828, tangentMode=RCTM_User, time=0.2779, value=1}
          {
           arriveTangent=0.0159
           arriveTangentWeight=0.2006
           interpMode=RCIM_Cubic
           leaveTangent=0.0159
           tangentMode=RCTM_User
           time=0.8796
           value=0.988
          }
          <at 0.9779 0>
         ]
        }}})
     Vector2D=&MakeVector2D(
      X=1
      Y=&FloatFromCurve(
       FloatCurve=DataInterface() {propertyValues {Curve {
          defaultValue=1
          keys [
           {
            arriveTangent=-0.182
            arriveTangentWeight=0.7464
            interpMode=RCIM_Cubic
            leaveTangent=-0.182
            leaveTangentWeight=0.7464
            tangentMode=RCTM_User
            time=0
            value=1
           }
           {
            arriveTangent=-0.3907
            arriveTangentWeight=0.2669
            interpMode=RCIM_Cubic
            leaveTangent=-0.3907
            leaveTangentWeight=0.2669
            tangentMode=RCTM_User
            time=1
            value=9
           }
          ]
         }}}
       CurveIndex=&Abs_Float(Float:NiagaraFloat=&DotProduct(
         Vector A=&ReturnCameraVector(
          Camera Vectors=::NewEnumerator4
          Coordinate Space=::Simulation
          Origin Position=$Particles.Position
          CameraQuery=DataInterface() {propertyValues {bRequireCurrentFrameData=false}}
         )
         Vector B=<0 0 1>
        ))
      )
     )
    )
    Scale Sprite Size Mode=::NewEnumerator2
   )
   &CameraOffset(Camera Offset Amount=&FloatFromCurve(
     FloatCurve=DataInterface() {propertyValues {Curve {
        defaultValue=1
        keys [
         {
          arriveTangent=-0.182
          arriveTangentWeight=0.7464
          interpMode=RCIM_Cubic
          leaveTangent=-0.182
          leaveTangentWeight=0.7464
          tangentMode=RCTM_User
          time=0.85
          value=-50
         }
         {
          arriveTangent=-0.3907
          arriveTangentWeight=0.2669
          interpMode=RCIM_Cubic
          leaveTangent=-0.3907
          leaveTangentWeight=0.2669
          tangentMode=RCTM_User
          time=1
          value=-50
         }
        ]
       }}}
     CurveIndex=&Abs_Float(Float:NiagaraFloat=&DotProduct(
       Vector A=&ReturnCameraVector(
        Camera Vectors=::NewEnumerator4
        Coordinate Space=::Simulation
        Origin Position=$Particles.Position
        CameraQuery=DataInterface() {propertyValues {bRequireCurrentFrameData=false}}
       )
       Vector B=<0 0 1>
      ))
    ))
   &ScaleColor(
    Scale Alpha=&FloatFromCurve(FloatCurve=DataInterface() {propertyValues {Curve {
        defaultValue=1
        keys [
         {interpMode=RCIM_Cubic, tangentMode=RCTM_User, time=0, value=0}
         {interpMode=RCIM_Cubic, tangentMode=RCTM_User, time=0.2, value=1}
         {arriveTangent=-0.0951, interpMode=RCIM_Cubic, leaveTangent=-0.0951, tangentMode=RCTM_User, time=0.8, value=1}
         {tangentMode=RCTM_User, time=1, value=0}
        ]
       }}})
    ScaleA=true
    ScaleRGB=true
   ) {enabled=false, module=ScaleColor002}
  ]
  rendererClasses [&NiagaraSpriteRendererProperties]
  renderers [{rendererClass=&NiagaraSpriteRendererProperties}]
 }
 Portal_Glow001 {
  bEnabled=false
  emitterUpdateScript [
   &EmitterState()
   &SpawnBurst_Instantaneous(Spawn Time=0.05, Use Spawn Probability=false, UseLoopCountLimit=false)
  ]
  particleSpawnScript [
   &InitializeParticle(
    Color=<rgba 1.199 2 0.3599 50>
    Lifetime=1.3
    Position Offset=<0 0 -1>
    Sprite Size=<90 90>
    Sprite Size Mode=::NewEnumerator4
    UsePositionOffset=true
    Write Material Random=true
   )
   /Game/Sprite/FN_SpriteAlignmentAndFacing(SpriteAlignment:Vector3f=<1 0 0>, SpriteFacing:Vector3f=<0 0 1>)
  ]
  particleUpdateScript [
   &SolveForcesAndVelocity(Clamp Velocity=false, Limit Acceleration=false)
   &ParticleState()
   &ScaleSpriteSize(Uniform Curve Sprite Scale=DataInterface() {propertyValues {Curve {keys [
        {tangentMode=RCTM_User, time=0, value=0}
        {arriveTangent=2.9121, interpMode=RCIM_Cubic, leaveTangent=2.9121, time=0.2237, value=0}
        {arriveTangent=4.1834, interpMode=RCIM_Cubic, leaveTangent=4.1834, time=0.3434, value=1}
        {arriveTangent=-3.195, interpMode=RCIM_Cubic, leaveTangent=-3.195, time=0.4627, value=1}
        {arriveTangent=-0.5127, interpMode=RCIM_Cubic, leaveTangent=-0.5127, tangentMode=RCTM_User, time=0.6564, value=0}
        {tangentMode=RCTM_User, time=1, value=0}
       ]}}})
   &ScaleColor(
    Scale Alpha=&FloatFromCurve(FloatCurve=DataInterface() {propertyValues {Curve {
        defaultValue=1
        keys [
         {interpMode=RCIM_Cubic, tangentMode=RCTM_User, time=0, value=0}
         {interpMode=RCIM_Cubic, tangentMode=RCTM_User, time=0.2, value=1}
         {arriveTangent=-0.0951, interpMode=RCIM_Cubic, leaveTangent=-0.0951, tangentMode=RCTM_User, time=0.8, value=1}
         {tangentMode=RCTM_User, time=1, value=0}
        ]
       }}})
    ScaleA=true
    ScaleRGB=true
   ) {enabled=false, module=ScaleColor002}
  ]
  rendererClasses [&NiagaraSpriteRendererProperties]
  renderers [{rendererClass=&NiagaraSpriteRendererProperties}]
 }
 Sparks {
  emitterUpdateScript [
   &EmitterState(Life Cycle Mode=::NewEnumerator1, Loop Behavior=::NewEnumerator1, Loop Delay=0.05, Loop Duration=1, UseLoopDelay=true)
   &SpawnBurst_Instantaneous(Spawn Count=12, Spawn Time=0.6, Use Spawn Probability=false)
   &SpawnBurst_Instantaneous(Spawn Count=8, Spawn Group=1, Spawn Time=0.6, Use Spawn Probability=false) {module=SpawnBurst_Instantaneous001}
  ]
  particleSpawnScript [
   &InitializeParticle(
    Color=<rgba 0.2833 1 0.2397 1>
    Lifetime=&Multiply_Float(
     A=&MergeHelper_FloatSwitch(0=1, 1=0.6, 2=?, 3=?, Index=&SpawnGroup(), Num=::NewEnumerator0)
     B=&RandomRangeFloat(Minimum=0.3, Maximum=1)
    )
    Sprite Size Max=<1.5 6>
    Sprite Size Min=<1.5 3>
    Sprite Size Mode=::NewEnumerator2
    UsePositionOffset=false
    Write Material Random=true
   )
   &ShapeLocation(
    Disc Coverage=1
    MaskSpawnGroup=false
    Offset=<0 5 0>
    Offset Mode=::NewEnumerator0
    Override Seed=false
    Ring Radius=&MergeHelper_FloatSwitch(0=35, 1=35, 2=?, 3=?, Index=&SpawnGroup(), Num=::NewEnumerator0)
    Rotation Axis=<0 0 1>
    Rotation Coordinate Space=::Simulation
    Rotation Mode=::NewEnumerator1
    Shape Primitive=::NewEnumerator4
    UseOwnerScale=true
   )
   &AddVelocityFromPoint(
    Module.Use Falloff Distance=false
    Origin Offset Coordinate Space=::Simulation
    Velocity Strength=&Multiply_Float(
     A=&Multiply_Float(
      A=&FN_FrontEndFloatModifier(Gameplay=0.7, Front End=0.35)
      B=&MergeHelper_FloatSwitch(0=1, 1=0.25, 2=?, 3=?, Index=&SpawnGroup(), Num=::NewEnumerator0)
     )
     B=&RandomRangeFloat(Minimum=75, Maximum=200)
    )
   )
   &CameraOffset(Camera Offset Amount=&Multiply_Float(A=30, B=$System.CameraOffsetScale))
   &AddVelocity(
    Rotation Coordinate Space=::Simulation
    Velocity=$Particles.ShapeLocation.ShapeBitangent
    Velocity Speed Scale=&RandomRangeFloat(Minimum=40, Maximum=80)
   ) {enabled=false}
   &AddVelocity(
    Rotation Coordinate Space=::Simulation
    Velocity=<0 0 -5>
    Velocity Speed Scale=&Multiply_Float(
     A=&RandomRangeFloat(Minimum=50, Maximum=150)
     B=&MergeHelper_FloatSwitch(0=1, 1=2, 2=?, 3=?, Index=&SpawnGroup(), Num=::NewEnumerator0)
    )
   ) {module=AddVelocity001}
  ]
  particleUpdateScript [
   &ParticleState()
   &ScaleColor(
    Scale Alpha=0.25
    Scale RGB=&VectorFromFloat(Value=&FloatFromCurve(
      FloatCurve=DataInterface() {propertyValues {Curve {keys [{interpMode=RCIM_Cubic, tangentMode=RCTM_User, time=0, value=10}, <at 1 1>]}}}
      Scale Curve=0.2
     ))
    ScaleA=true
    ScaleRGB=true
   ) {module=ScaleColor001}
   &Drag(
    Drag=&Multiply_Float(
     A=&FloatFromCurve(
      FloatCurve=DataInterface() {propertyValues {Curve {keys [<at 0 0.5>, {arriveTangent=0.5, interpMode=RCIM_Cubic, leaveTangent=0.5, time=0.3, value=0.5}, <at 1 1 RCIM_Cubic>]}}}
      Scale Curve=15
     )
     B=1
    )
    Use Linear Drag=true
    Use Rotational Drag=false
   )
   /Niagara/Modules/Update/Forces/CurlNoiseForce(
    Noise Frequency=1
    Noise Strength=&FloatFromCurve(FloatCurve=DataInterface() {propertyValues {Curve {keys [<at 0 1>,<at 1 0>]}}}, Scale Curve=800)
    Pan Noise=true
    Pan Noise Field=<0.1 0.1 0.5>
   )
   &GravityForce(Gravity=&Multiply_VectorByFloat(
     Vector=&RandomRangeVector(Minimum=<0 0 800>, Maximum=<0 0 -500>)
     Float=&MergeHelper_FloatSwitch(0=1, 1=0.2, 2=?, 3=?, Index=&SpawnGroup(), Num=::NewEnumerator0)
    )) {module=GravityForce001}
   &SolveForcesAndVelocity(Clamp Velocity=false, Limit Acceleration=false)
   &ScaleSpriteSizeBySpeed(Max Scale Factor=<1 6>, Min Scale Factor=<0.5 0.1>, Velocity Threshold=800)
   &ScaleSpriteSize(Scale Factor=&MakeVector2D(
     X=&FloatFromCurve(FloatCurve=DataInterface() {propertyValues {Curve {keys [<at 0 0.5>,<at 0.1 1>,<at 1 0.5>]}}})
     Y=&FloatFromCurve(FloatCurve=DataInterface() {propertyValues {Curve {keys [<at 0 1>,<at 1 0>]}}})
    ))
   &ScaleSpriteSize(
    Scale Sprite Size Mode=::NewEnumerator0
    Uniform Scale Factor=&FN_FrontEndFloatModifier(Gameplay=1, Front End=$User.FrontEndScale)
   ) {module=ScaleSpriteSize001}
   &FN_Sprite_Attributes(
    Color=&MakeLinearColorFromVectorAndFloat(
     "Vector (RGB)"=&MakeVectorFromLinearColorRGB(Color=&ColorFromCurve(ColorCurve=DataInterface() {propertyValues {
         AlphaCurve {keys [<at 0 1 RCIM_Cubic>,<at 1 0 RCIM_Cubic>]}
         BlueCurve {keys [<at 0 0.3769 RCIM_Cubic>,<at 0.9637 0.0841 RCIM_Cubic>]}
         GreenCurve {keys [<at 0 2 RCIM_Cubic>,<at 0.9637 0.5 RCIM_Cubic>]}
         RedCurve {keys [<at 0 0.581 RCIM_Cubic>,<at 0.9637 0.1036 RCIM_Cubic>]}
        }}))
     "Float (Alpha)"=&FloatFromCurve(FloatCurve=DataInterface() {propertyValues {Curve {keys [<at 0 1>,<at 0.8 0>]}}}, Scale Curve=0.25)
    )
    Create Color Attribute=true
    Create SpriteSize Attribute=true
    SpriteSize=&Multiply_Vector2D_ByFloat(
     Float=&FN_FrontEndFloatModifier(Gameplay=0.85, Front End=0.85)
     Vector2D=&Multiply_Vector2D(
      A=<5 15>
      B=&MakeVector2D(
       X=&FloatFromCurve(FloatCurve=DataInterface() {propertyValues {Curve {keys [<at 0 0.5>,<at 0.1 1>,<at 1 0.5>]}}})
       Y=&FloatFromCurve(FloatCurve=DataInterface() {propertyValues {Curve {keys [<at 0 1>,<at 1 0>]}}})
      )
     )
    )
    Write to Module Namespace=true
   )
  ]
  rendererClasses [&NiagaraSpriteRendererProperties]
  renderers [
   {rendererClass=&NiagaraSpriteRendererProperties}
   {rendererClass=&NiagaraSpriteRendererProperties, rendererIndex=1}
  ]
 }
}
summary {
 systemName=NS_Emote_LentilAnise_PortalB
 emitters [
  {emitterName=Portal, rendererClasses [&NiagaraSpriteRendererProperties, &NiagaraLightRendererProperties]}
  {emitterName=Sparks, rendererClasses [&NiagaraSpriteRendererProperties]}
  {emitterName=Air_Trail_Up, rendererClasses [&NiagaraRibbonRendererProperties]}
  {emitterName=Flash001, rendererClasses [&NiagaraSpriteRendererProperties, &NiagaraMeshRendererProperties]}
  {emitterName=ElectricSprites, rendererClasses [&NiagaraSpriteRendererProperties]}
  {bEnabled=false, emitterName=Portal_Glow, rendererClasses [&NiagaraSpriteRendererProperties]}
  {bEnabled=false, emitterName=Portal_Glow001, rendererClasses [&NiagaraSpriteRendererProperties]}
 ]
 userVariables [
  User.bIsFrontEnd {type=NiagaraBool}
  User.bIsFrontEndPreview {type=NiagaraBool}
  User.FrontEndScale {type=NiagaraFloat, defaultValue=0.5, description=""}
  User.PortalScale {type=NiagaraFloat, defaultValue=4, description=""}
 ]
}
